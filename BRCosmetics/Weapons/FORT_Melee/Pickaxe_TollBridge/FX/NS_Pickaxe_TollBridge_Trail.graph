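# Niagara ribbon trail system for the TollBridge pickaxe cosmetic, drawing short-lived anim trails between bones or sockets on a supplied skeletal mesh.
# FN_AnimTrail drives three ribbon renderers with red and off-white colors tinted by time of day, the off-white carried in a separate WhiteRibbon parameter that fades faster.
# FN_AnimTrail001 adds two dark, untinted trails on socket indices 2 and 3 through two ribbon renderers, fading by alpha curve.
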NiagaraSystem /BRCosmetics/Weapons/FORT_Melee/Pickaxe_TollBridge/FX/NS_Pickaxe_TollBridge_Trail
dependencies {
 usedDataInterfaces [&NiagaraDataInterfaceCurve]
 usedDynamicInputs []
 usedModules [
  &self:SystemSpawnScript.SystemScriptSource.SystemScriptGraph.NiagaraNodeAssignment_0.SetVariables_#1
  &SystemState
 ]
}
emitters {
 FN_AnimTrail {
  emitterSpawnScript [&FN_RibbonIDOffsetUtility()]
  emitterUpdateScript [
   &EmitterState()
   &FN_SpawnAnimTrail()
  ]
  particleSpawnScript [
   &FN_InitializeAnimTrailBoneIndecies()
   &FN_InitializeAnimTrail(
    Color=&MultiplyLinearColor(
     Linear Color=<rgba 0.5271 0.0183 0.0372 0.6>
     Scale Factor=&MakeVector4_FromLinearColor(LinearColor=&FN_Select_Color_from_Time_Of_Day(Morning=<rgba 0.7937 0.4824 0.3508 1>, Day=<rgba 1 1 1 1>, Evening=<rgba 1 1 1 1>, Night=<rgba 0.3116 0.277 0.5437 1>))
    )
    Lifetime=0.18
    SkeletalMesh=$System.Skeletal Mesh
   )
   &AddVelocity(Velocity=<0 1500 0>) {enabled=false}
   /SetVariables(
    Particles.Dy2:Vector4f=<1 1.271 0 0>
    Particles.WhiteRibbon:LinearColor=&MultiplyLinearColor(
     Linear Color=<rgba 0.7521 0.6907 0.6771 0.75>
     Scale Factor=&MakeVector4_FromLinearColor(LinearColor=&FN_Select_Color_from_Time_Of_Day(Morning=<rgba 0.8562 0.6083 0.4867 1>, Day=<rgba 1 1 1 1>, Evening=<rgba 1 1 1 1>, Night=<rgba 0.3842 0.3928 0.7437 1>))
    )
   ) {isSetParameters=true}
  ]
  particleUpdateScript [
   &SolveForcesAndVelocity(Clamp Velocity=false, Limit Acceleration=false)
   &FN_UpdateAnimTrail(Locked Particle RibbonWidth=1, SkeletalMesh=$System.Skeletal Mesh) {module=FN_UpdateAnimTrail001}
   &ScaleColor(
    Scale Alpha=&FloatFromCurve(FloatCurve=DataInterface() {propertyValues {Curve {keys [{interpMode=RCIM_Cubic, leaveTangent=0.0155, tangentMode=RCTM_Break, time=0, value=1}, <at 0.6 0>]}}})
    ScaleA=true
    ScaleRGB=true
   )
   /SetVariables(Particles.WhiteRibbon:LinearColor=&MultiplyLinearColor(
     Linear Color=$Particles.Initial.WhiteRibbon
     Scale Factor=&MakeVector4(
      X=1
      Y=1
      Z=1
      W=&FloatFromCurve(FloatCurve=DataInterface() {propertyValues {Curve {keys [
           {interpMode=RCIM_Cubic, leaveTangent=-5.7841, tangentMode=RCTM_Break, time=0, value=1}
           {arriveTangent=-0.1996, tangentMode=RCTM_User, time=0.4, value=0}
          ]}}})
     )
    )) {isSetParameters=true}
  ]
  rendererClasses [&NiagaraRibbonRendererProperties]
  renderers [
   {rendererClass=&NiagaraRibbonRendererProperties}
   {rendererClass=&NiagaraRibbonRendererProperties, rendererIndex=1}
   {rendererClass=&NiagaraRibbonRendererProperties, rendererIndex=2}
  ]
 }
 FN_AnimTrail001 {
  emitterSpawnScript [&FN_RibbonIDOffsetUtility()]
  emitterUpdateScript [
   &EmitterState()
   &FN_SpawnAnimTrail(Number of Anim Trails=2)
  ]
  particleSpawnScript [
   &FN_InitializeAnimTrailBoneIndecies(First Bone or Socket Index=2, Second Bone or Socket Index=3)
   &FN_InitializeAnimTrailBoneIndecies(First Bone or Socket Index=3, Second Bone or Socket Index=2) {module=FN_InitializeAnimTrailBoneIndecies001}
   &FN_InitializeAnimTrail(AnimTrailWidth=0.9, Color=<rgba 0.0313 0.0303 0.0282 2.5>, Lifetime=0.18, SkeletalMesh=$System.Skeletal Mesh)
   &AddVelocity(Velocity=<0 1500 0>) {enabled=false}
  ]
  particleUpdateScript [
   &SolveForcesAndVelocity(Clamp Velocity=false, Limit Acceleration=false)
   &FN_UpdateAnimTrail(Locked Particle RibbonWidth=1, SkeletalMesh=$System.Skeletal Mesh) {module=FN_UpdateAnimTrail001}
   &ScaleColor(
    Scale Alpha=&FloatFromCurve(FloatCurve=DataInterface() {propertyValues {Curve {keys [
         {interpMode=RCIM_Cubic, leaveTangent=-2.4186, tangentMode=RCTM_Break, time=0, value=1}
         {arriveTangent=-0.3468, tangentMode=RCTM_User, time=1, value=0}
        ]}}})
    ScaleA=true
    ScaleRGB=true
   )
  ]
  rendererClasses [&NiagaraRibbonRendererProperties]
  renderers [
   {rendererClass=&NiagaraRibbonRendererProperties}
   {rendererClass=&NiagaraRibbonRendererProperties, rendererIndex=1}
  ]
 }
}
summary {
 systemName=NS_Pickaxe_TollBridge_Trail
 emitters [
  {emitterName=FN_AnimTrail, rendererClasses [&NiagaraRibbonRendererProperties]}
  {emitterName=FN_AnimTrail001, rendererClasses [&NiagaraRibbonRendererProperties]}
 ]
 userVariables [
  User.RibbonIDOffset {type=NiagaraInt32}
  User.SkeletalMesh {
   type=NiagaraDataInterfaceSkeletalMesh
   defaultValue {
    struct=NiagaraExt_VariableValue_DataInterface
    value {dataInterface=&self:NiagaraDataInterfaceSkeletalMesh_1}
   }
  }
 ]
}
